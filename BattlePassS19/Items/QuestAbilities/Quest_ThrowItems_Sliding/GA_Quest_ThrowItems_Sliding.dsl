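; Quest gameplay ability that credits the player for throwing shield potions, medkits or bandages, apparently while sliding.
; On event activation, it stores the instigator tags, resolves the avatar pawn's quest manager, and checks the pawn's owned tags against TagsToCheck.
; The first item tag set matching the instigator tags sends its custom stat event.
; The ability then ends, except on invalid pawn or controller paths, which only print an error.
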(event Ability|EventActivateAbilityFromEvent
  (params (EventData :type Gameplay%20Event%20Data%20Structure%20%28by%20ref%29))
  (bind (_eventtag _instigator _target _optionalobject _optionalobject2 _contexthandle _instigatortags) (Utilities|Struct|BreakGameplayEventData EventData))
  (bind _returnvalue (CallFunction|GetAvatarasFortPlayerPawn))
  (|CallParentFunction EventData)
  (Variables|Default|SetInstigator%20Tags _instigatortags)
  (Utilities|IsValid _returnvalue
    (:Is%20Valid
      (bind _output_get (Variables|GAVars|Data|SetOwningPlayerPawn _returnvalue))
      (Utilities|IsValid (Fortnite|GetFortPlayerControllerfromActor _output_get)
        (:Is%20Valid
          (Variables|GAVars|Data|SetOwningPlayerQuestManager (Quests|GetQuestManager (Fortnite|GetFortPlayerControllerfromActor _output_get) "Invalid"))
          (bind _returnvalue_2 (MissionHelpers|HasTags (GameplayTags|GetOwnedGameplayTags (Variables|GAVars|Data|GetOwningPlayerPawn) :!class /Script/GameplayTags.GameplayTagAssetInterface) (Variables|Default|GetTagsToCheck) false false))
          (if _returnvalue_2
            (bind _returnvalue_3 (MissionHelpers|HasTags (Variables|Default|GetInstigator%20Tags) (Variables|Default|GetShieldPotionTags) false false))
            (if _returnvalue_3
              (Quest|SendComplexCustomStatEvent (Variables|GAVars|Data|GetOwningPlayerQuestManager) 0 (Variables|Default|GetShieldPotStatEvent) "(GameplayTags=)" 1)
              (Ability|EndAbility)
              (else
                (bind _returnvalue_4 (MissionHelpers|HasTags (Variables|Default|GetInstigator%20Tags) (Variables|Default|GetMedkitTags) false false))
                (if _returnvalue_4
                  (Quest|SendComplexCustomStatEvent (Variables|GAVars|Data|GetOwningPlayerQuestManager) 0 (Variables|Default|GetMedkitStatEvent) "(GameplayTags=)" 1)
                  (Ability|EndAbility)
                  (else
                    (bind _returnvalue_5 (MissionHelpers|HasTags (Variables|Default|GetInstigator%20Tags) (Variables|Default|GetBandageTags) false false))
                    (if _returnvalue_5
                      (Quest|SendComplexCustomStatEvent (Variables|GAVars|Data|GetOwningPlayerQuestManager) 0 (Variables|Default|GetBandageStatEvent) "(GameplayTags=)" 1)
                      (Ability|EndAbility)
                      (else
                        (Ability|EndAbility)))))))
            (else
              (Ability|EndAbility))))
        (:Is%20Not%20Valid
          (Development|PrintString "@@@ ActivateAbilityFromEvent [GA_Quest_ThrowItems_Sliding] GET PLAYER CONTROLLER FAILED"))))
    (:Is%20Not%20Valid
      (Development|PrintString "@@@ ActivateAbilityFromEvent [GA_Quest_ThrowItems_Sliding]: CAST TO FORT PAWN FAILED"))))
